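# Looping UI sound cue for Rocket Drill refueling, routed to the non-critical in-game UI sound class.
# A continuous modulator drives the pitch of the fuel charge loop from the Charge parameter, with a lowered minimum output.
# An Enveloper at the root ramps volume from quiet to full over the first three quarters of a second.
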SoundCue /RocketDrillGameplay/Assets/Audio/SC_RocketDrill_UI_Refueling
root=Enveloper_1
nodes [
 {
  id=Enveloper_1
  children [ModulatorContinuous_0]
  properties {
   pitchCurve {editorCurveData {keys [{time=0}]}}
   volumeCurve {editorCurveData {keys [
      {arriveTangent=2.206, interpMode=RCIM_Cubic, leaveTangent=2.206, tangentMode=RCTM_User, time=0, value=0.1}
      {arriveTangent=1.7729, interpMode=RCIM_Cubic, leaveTangent=1.7729, tangentMode=RCTM_User, time=0.75, value=1}
     ]}}
  }
 }
 {id=ModulatorContinuous_0, children [WavePlayer_0], properties {pitchModulationParams {default=0.014, minOutput=0.75, parameterName=Charge}}}
 {id=WavePlayer_0, properties {bLooping=true, soundWaveAssetPtr=/RocketDrillGameplay/Assets/Audio/waves/RocketDrill_FuelCharge_Loop_01}}
]
properties {maxDistance=2097152, soundClassObject=&Fort_UI_InGame_NonCritical, volumeMultiplier=1}
