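# Sound cue for the Rocket Drill's refueling-complete UI sound, routed to the non-critical in-game UI sound class.
# A mixer layers a fuel-charge stop wave, boosted in the mix, with a charge-complete wave.
# Each wave passes through its own modulator for random variation, with the completion layer shifted to a higher pitch range.
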SoundCue /RocketDrillGameplay/Assets/Audio/SC_RocketDrill_UI_Refueling_Complete
root=Mixer_0
nodes [
 {id=Mixer_0, children [Modulator_0,Modulator_1], properties {inputVolume [1.3,1]}}
 {id=Modulator_0, children [WavePlayer_0]}
 {id=Modulator_1, children [WavePlayer_2], properties {pitchMax=1.2, pitchMin=1.1}}
 {id=WavePlayer_0, properties {soundWaveAssetPtr=/RocketDrillGameplay/Assets/Audio/waves/RocketDrill_FuelCharge_Stop_01}}
 {id=WavePlayer_2, properties {soundWaveAssetPtr=/RocketDrillGameplay/Assets/Audio/waves/RocketDrill_FuelCharge_Complete}}
]
properties {maxDistance=2097152, soundClassObject=&Fort_UI_InGame_NonCritical, volumeMultiplier=1}
