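# Translucent UI material for the Rocket Drill reticle dot, which blends between a circle and a chevron shape through CircleOrChevron.
# The chevron comes from the maximum of two rotated distance gradients on symmetry-mapped UVs, positioned and sized by ArrowOffset and ArrowScale; the circle is an SDF.
# Two antialiased samples of the blended field, one with a shape-dependent threshold, are averaged and scaled by MasterOpacity.
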Material /RocketDrillGameplay/UI/M_UI_Ret_RocketDrill_Dot
graph {
 properties {bAutomaticallySetUsageInEditor=false, blendMode=BLEND_Translucent, materialDomain=MD_UI, usedShadingModels=Default Lit}
 nodes [
  n1 = ScalarParameter() {defaultValue=1, parameterName=ArrowOffset}
  n2 = ScalarParameter() {defaultValue=1, parameterName=ArrowScale}
  n3 = MaterialFunctionCall!(
   UV=TextureCoordinate()
   Offset=AppendVector(A=Constant(), B=LinearInterpolate(Alpha=@n1) {constA=-1.1, constB=-0.3})
   Rotation
   Scale=AppendVector(A=LinearInterpolate(Alpha=@n2) {constA=0.125, constB=0.2}, B=Constant() {r=0.2})
   MirrorY
  ) {materialFunction=/FortUILibrary/MaterialFunctions/Basics/MF_UI_Symmetry_2x1_DEPRECATED}
  n4 = Constant() {r=-0.4}
  n5 = Rotator(Coordinate=@n3.Result, Time=@n4) {speed=3.1416}
  n6 = Rotator(Coordinate=Multiply(A=@n3.Result) {constB=-1}, Time=Add(A=@n4) {constB=0.075}) {speed=3.1416}
  n7 = ScalarParameter() {defaultValue=1, parameterName=CircleOrChevron}
  n8 = LinearInterpolate(
   A=Max(A=ComponentMask(None=@n5) {r=true}, B=ComponentMask(None=@n6) {r=true})
   B=MaterialFunctionCall!(UVs, Radius=Constant() {r=0.075}) {materialFunction=/FortUILibrary/MaterialFunctions/Shapes/SDFs/MF_SDF_Circle}.SDF
   Alpha=@n7
  )
  n9 = LinearInterpolate(A=Constant() {r=0.07}, B=Constant() {r=0.02}, Alpha=@n7)
  Multiply(
   A=LinearInterpolate(
    A=MaterialFunctionCall(DistanceField=@n8) {materialFunction=&MF_UI_AntiAliasing}.Inner
    B=MaterialFunctionCall(DistanceField=@n8, Threshold=@n9) {materialFunction=&MF_UI_AntiAliasing}.Inner
   )
   B=LinearInterpolate(Alpha=ScalarParameter() {defaultValue=1, parameterName=MasterOpacity}) {constB=0.85}
  )
 ]
}
