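# Translucent UI material that procedurally draws the center-arrow frame of the RocketDrill reticle.
# Rotated, center-scaled UVs feed two-sided reticle UV functions and triangle SDFs, with a second, offset triangle subtracted via Max of the negated field to cut each arrow shape.
# Two anti-aliased thresholds of that field are blended for the edge, then multiplied by CenterOpacity.
# CenterArrowOffset drives the UV scale of both triangles.
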Material /RocketDrillGameplay/UI/M_UI_Ret_RocketDrill_Frame
graph {
 properties {bAutomaticallySetUsageInEditor=false, blendMode=BLEND_Translucent, materialDomain=MD_UI, usedShadingModels=Default Lit}
 nodes [
  n1 = MaterialFunctionCall(UVs=TextureCoordinate()) {materialFunction=&CustomRotator}
  n2 = ScalarParameter() {defaultValue=0.9, parameterName=CenterArrowOffset}
  n3 = AppendVector(A=Add(B=@n2) {constA=-0.4}, B=@n2)
  n4 = MaterialFunctionCall(UVs=@n1.Rotated Values, Texture Scale=@n3) {materialFunction=&ScaleUVsByCenter}
  n5 = Constant() {r=2}
  n6 = Constant() {r=0.4}
  n7 = Constant() {r=3}
  n8 = Constant2Vector() {g=0.9}
  n9 = MaterialFunctionCall!(
   Reticle UVs=MaterialFunctionCall(UVs=@n1.Rotated Values, Texture Scale=Add(A=@n3, B=@n8)) {materialFunction=&ScaleUVsByCenter}.UVs
   Reticle Sides=@n5
   Reticle Scale
   Reticle Rotation
   Prong Scale
   Prong Rotation
   Spread Offset=@n6
   Positional Offset=Constant() {r=-0.3}
   "Spread Input Range [Min,Max]"
   "Spread Remapped Range [Min,Max]"
  ) {materialFunction=/FortUILibrary/Systems/Reticles/MaterialFunctions/MF_UI_Reticle_UVs}
  n10 = Constant() {r=0.75}
  n11 = Max(
   A=MaterialFunctionCall!(
    UVs=MaterialFunctionCall(Reticle UVs=@n4.UVs, Reticle Sides=@n5, Spread Offset=@n6) {materialFunction=&MF_UI_Reticle_UVs}.UVs
    Radius
    Sides=@n7
    InnerAngle
   ) {materialFunction=/FortUILibrary/MaterialFunctions/Shapes/SDFs/MF_SDF_RegularPolygon}.SDF
   B=Multiply(A=MaterialFunctionCall(UVs=@n9.UVs, Radius=@n10, Sides=@n7) {materialFunction=&MF_SDF_RegularPolygon}.SDF) {constB=-1}
  )
  n12 = Constant() {r=0.02}
  Multiply(
   A=LinearInterpolate(
    A=MaterialFunctionCall(DistanceField=@n11) {materialFunction=&MF_UI_AntiAliasing}.Inner
    B=MaterialFunctionCall(DistanceField=@n11, Threshold=@n12) {materialFunction=&MF_UI_AntiAliasing}.Inner
   ) {constAlpha=0.4}
   B=ScalarParameter() {defaultValue=1, parameterName=CenterOpacity}
  )
 ]
}
