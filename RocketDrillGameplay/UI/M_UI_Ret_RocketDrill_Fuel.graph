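# Translucent UI material for the Rocket Drill reticle's fuel gauge, drawing a radial arc bar inside a procedural frame.
# A circle outline masked by a radial arc forms the bar, with a Progress parameter and threshold-based red-to-yellow coloring.
# The frame and its shadow are built texture-free from circle and box SDFs, mirrored rotated cuts and anti-aliasing.
# isActiveSwitch enables a sine pulse in frame color, and opacity parameters control blending.
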Material /RocketDrillGameplay/UI/M_UI_Ret_RocketDrill_Fuel
graph {
 properties {bAutomaticallySetUsageInEditor=false, blendMode=BLEND_Translucent, materialDomain=MD_UI, usedShadingModels=Default Lit}
 nodes [
  n1 = NamedRerouteUsage()
  n2 = ScalarParameter() {defaultValue=0.85, parameterName=BarOpacity}
  Multiply(
   A=LinearInterpolate(A=NamedRerouteUsage(), B=@n1, Alpha=Multiply(A=@n1, B=@n2))
   B=ScalarParameter() {defaultValue=1, parameterName=MasterOpacity}
  )
  NamedRerouteDeclaration(None=MaterialFunctionCall() {materialFunction=&GetUserInterfaceUV}.Normalized UV) {name=UV, nodeColor=<rgba 0 0 0 1>}
  NamedRerouteDeclaration(None=MaterialFunctionCall(
    Size=Constant() {r=0.635}
    OutlineThickness=ScalarParameter() {defaultValue=0.02, group=AmmoCounter, parameterName=RadialThickness, sliderMax=1}
   ) {materialFunction=&MF_UI_Circle}.Outline) {name=Mask_Radial, nodeColor=<rgba 0 0 0 1>}
  n3 = MaterialFunctionCall!(
   UV
   "Arc Start Angle (0-1)"=ScalarParameter() {defaultValue=0.255, parameterName=Start Angle, sliderMax=1, sortPriority=31}
   "Arc End Angle (0-1)"=ScalarParameter() {defaultValue=0.505, parameterName=End Angle, sliderMax=1}
   "Fill Ratio (0-1)"=OneMinus(None=NamedRerouteUsage())
   Fill Clockwise=Constant()
  ) {materialFunction=/FortUILibrary/MaterialFunctions/Basics/MF_UI_RadialArc}
  NamedRerouteDeclaration(None=Multiply(A=NamedRerouteUsage(), B=SmoothStep(Value=@n3.Radial Fill) {constMax=0.001})) {name=Mask_OverheatProgress, nodeColor=<rgba 0 0 0 1>}
  n4 = TextureCoordinate()
  n5 = MaterialFunctionCall!(UVs=ConstantBiasScale(None=@n4) {bias=-0.5, scale=1}, Size=Constant() {r=1}, CornerRoundness, KeepAspectRatio, PixelSize) {materialFunction=/FortUILibrary/MaterialFunctions/Shapes/SDFs/MF_SDF_Box}
  NamedRerouteDeclaration(None=Multiply(A=@n5.SDF) {constB=-1}) {name=BLQuad2, nodeColor=<rgba 0 0 0 1>}
  n6 = Constant() {r=1}
  n7 = Time()
  n8 = Add(A=ConstantBiasScale(None=Sine(None=@n7) {period=0.75}) {bias=0.5, scale=0.25}, B=NamedRerouteUsage())
  n9 = Ceil(None=ScalarParameter() {parameterName=isActiveSwitch})
  n10 = NamedRerouteUsage()
  NamedRerouteDeclaration(None=LinearInterpolate(A=LinearInterpolate(A=@n6, B=@n8, Alpha=@n9), B=LinearInterpolate(Alpha=@n10), Alpha=NamedRerouteUsage())) {name=Outer Frame Color, nodeColor=<rgba 0 0 0 1>}
  n11 = Constant4Vector() {constant=<rgba 1 0 0 0>}
  n12 = NamedRerouteUsage()
  n13 = Constant4Vector() {constant=<rgba 0.9059 0.7216 0.0353 1>}
  n14 = LinearInterpolate(A=@n11.RGBA, B=@n13.RGBA, Alpha=@n12)
  NamedRerouteDeclaration(None=If(A=@n12, "A > B"=If(A=@n12, "A > B"=@n13.RGBA, "A < B"=@n14) {constB=0.75}, "A < B"=@n11.RGBA) {constB=0.25}) {name=Overheat Color, nodeColor=<rgba 0 0 0 1>}
  n15 = MaterialFunctionCall!(UVs, Radius=Constant() {r=0.7}) {materialFunction=/FortUILibrary/MaterialFunctions/Shapes/SDFs/MF_SDF_Circle}
  n16 = Abs(None=ConstantBiasScale(None=@n4) {bias=-0.5, scale=2})
  n17 = Min(A=NamedRerouteUsage(), B=Min(A=ComponentMask(None=@n16) {g=true}, B=ComponentMask(None=@n16) {r=true}))
  n18 = MaterialFunctionCall!(UVs=Rotator(Coordinate=@n4, Time=Constant() {r=0.5}) {speed=3.1416}, Sides=Constant() {r=4}) {materialFunction=/FortUILibrary/MaterialFunctions/Basics/MF_UI_UVRadialMirror}
  n19 = Constant() {r=0.15}
  n20 = Rotator(Coordinate=@n18.UVs, Time=@n19) {speed=3.1416}
  n21 = Constant() {r=-0.085}
  n22 = Rotator(Coordinate=@n18.UVs, Time=OneMinus(None=@n19)) {speed=3.1416}
  n23 = Min(A=ComponentMask(None=Add(A=@n20, B=@n21) {constB=-0.1}) {r=true}, B=ComponentMask(None=Add(A=@n22, B=@n21) {constB=-0.1}) {r=true})
  n24 = Constant() {r=0.25}
  n25 = ComponentMask(None=Add(A=Rotator(Coordinate=@n4, Time=@n24) {speed=3.1416}) {constB=-0.65}) {g=true}
  n26 = Constant() {r=0.02}
  n27 = Min(
   A=Max(A=Max(A=@n17, B=@n23), B=Multiply(A=@n25) {constB=-2})
   B=Max(
    A=MaterialFunctionCall!(DistanceField=@n15.SDF, Threshold, Thickness=@n26, Alignment, Smoothing) {materialFunction=/FortUILibrary/MaterialFunctions/Shapes/Modifiers/MF_UI_ShapeOutline}.SDFOutline
    B=@n5.SDF
   )
  )
  n28 = NamedRerouteDeclaration(None=Multiply(A=Max(A=@n15.SDF, B=@n27)) {constB=3}) {name=OuterFrameSDF, nodeColor=<rgba 0 0 0 1>}
  n29 = NamedRerouteDeclaration(None=MaterialFunctionCall(DistanceField=@n28, Threshold=Constant() {r=0.01}) {materialFunction=&MF_UI_AntiAliasing}.Inner) {name=Frame, nodeColor=<rgba 0 0 0 1>}
  n30 = MaterialFunctionCall(DistanceField=@n28, Threshold=Constant() {r=0.03}) {materialFunction=&MF_UI_AntiAliasing}
  NamedRerouteDeclaration(None=LinearInterpolate(A=Multiply(A=@n29) {constB=2}, B=NamedRerouteDeclaration(None=@n30.Inner) {name=FrameShadow, nodeColor=<rgba 0 0 0 1>}) {constAlpha=0.6}) {name=CombinedFrame, nodeColor=<rgba 0 0 0 1>}
  NamedRerouteDeclaration(None=ScalarParameter() {defaultValue=1, group=Color, parameterName=Progress, sliderMax=1}) {name=Progress, nodeColor=<rgba 0 0 0 1>}
  NamedRerouteUsage()
 ]
}
