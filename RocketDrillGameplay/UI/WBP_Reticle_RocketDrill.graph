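# Rocket Drill weapon reticle widget showing center and bottom markers, a fuel meter and a timed "Hold to Travel Further" hint.
# An async attribute listener drives the meter material's Progress and a refueling sound with a Charge parameter, plus a completion sound.
# Active-tag listeners play or reverse the boost animation and switch the meter's active state.
# A gameplay event triggers the hint, with debug prints remaining; cooldown changes hide the center dot.
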WidgetBlueprint /RocketDrillGameplay/UI/WBP_Reticle_RocketDrill
animations [
 Anim_InRange {
  binding_count=1
  length_seconds=0.35
  object=&self:Anim_InRange
  sequence {bindings [CenterDot {
     id=#1
     possessed_class=Image
     tracks [{
       type=MovieSceneWidgetMaterialTrack
       sections [{type=MovieSceneParameterSection, channels [AnimInRange {type=MovieSceneScriptingFloatChannel, keys [[0,0],[7,1]]}]}]
      }]
    }]}
 }
 Anim_OnBoosting {
  binding_count=3
  length_seconds=0.3334
  object=&self:Anim_OnBoosting
  sequence {
   bindings [
    BottomDot {
     id=#2
     possessed_class=Image
     tracks [
      {
       type=MovieSceneWidgetMaterialTrack
       sections [{
         type=MovieSceneParameterSection
         end_frame=10
         start_frame=0
         channels [
          ArrowOffset {type=MovieSceneScriptingFloatChannel, keys [[0,0],[9,1],[10,1]]}
          MasterOpacity {type=MovieSceneScriptingFloatChannel, keys [[3,0],[9,1],[10,1]]}
          CircleOrChevron {type=MovieSceneScriptingFloatChannel, keys [[0,1],[9,0]]}
         ]
        }]
      }
      {
       type=MovieSceneColorTrack
       display_name=Color and Opacity
       sections [{
         type=MovieSceneColorSection
         end_frame=1
         start_frame=0
         channels [
          Color.R {type=MovieSceneScriptingFloatChannel, const [1,1]}
          Color.G {type=MovieSceneScriptingFloatChannel, const [1,1]}
          Color.B {type=MovieSceneScriptingFloatChannel, const [1,1]}
          Color.A {type=MovieSceneScriptingFloatChannel, keys [[0,0],[1,1]]}
         ]
        }]
      }
     ]
    }
    CenterFrame {
     id=#3
     possessed_class=Image
     tracks [{
       type=MovieSceneWidgetMaterialTrack
       sections [{
         type=MovieSceneParameterSection
         end_frame=10
         start_frame=0
         channels [
          CenterOpacity {type=MovieSceneScriptingFloatChannel, keys [[0,1],[10,0]]}
          CenterArrowOffset {type=MovieSceneScriptingFloatChannel, const [0.9,1]}
         ]
        }]
      }]
    }
    CenterDot {
     id=#4
     possessed_class=Image
     tracks [
      {
       type=MovieSceneWidgetMaterialTrack
       sections [{
         type=MovieSceneParameterSection
         end_frame=10
         start_frame=0
         channels [CircleOrChevron {type=MovieSceneScriptingFloatChannel, keys [[0,1],[9,0]]}]
        }]
      }
      {
       type=MovieScene2DTransformTrack
       sections [{
         type=MovieScene2DTransformSection
         end_frame=10
         start_frame=0
         channels [
          Translation.X {type=MovieSceneScriptingFloatChannel, default=0}
          Translation.Y {type=MovieSceneScriptingFloatChannel, default=0}
          Angle {type=MovieSceneScriptingFloatChannel, default=0}
          Scale.X {type=MovieSceneScriptingFloatChannel, keys [[0,1],[9,0.25]]}
          Scale.Y {type=MovieSceneScriptingFloatChannel, keys [[0,1],[9,0.25]]}
          Shear.X {type=MovieSceneScriptingFloatChannel, default=0}
          Shear.Y {type=MovieSceneScriptingFloatChannel, default=0}
         ]
        }]
      }
     ]
    }
   ]
   display_rate {numerator=30}
  }
 }
 Anim_OnCooldown {binding_count=0, length_seconds=5, object=&self:Anim_OnCooldown, sequence {bindings []}}
]
graphs [
 EventGraph {nodes [
   n1 = HideHoldText() {is=title K2Node_CustomEvent}
   > SetVisibility(
    self=[
     Get HoldPrompt_Text1() -> (HoldPrompt_Text1:Common Text Object Reference)
     Get HoldPrompt_Text2() -> (HoldPrompt_Text2:Common Text Object Reference)
    ]
    InVisibility=Hidden
   )
   > n2 = PrintString!(execute, InString="HIDE!", bPrintToScreen, bPrintToLog, TextColor, Duration=20, Key) -> (then)
   Do Once(execute=@n11.EventReceived, Reset=@n2.then)
   > Clear and Invalidate Timer by Handle(Handle=Get HoldVisibilityTimer() -> (HoldVisibilityTimer:Timer Handle Structure))
   > SetVisibility(self=[Get HoldPrompt_Text2(), Get HoldPrompt_Text1()], InVisibility=Visible)
   > PrintString(InString="SHOW!", Duration=20)
   > n3 = Set Timer by Event(Delegate=@n1, Time=2, bLooping=true)
   > Set HoldVisibilityTimer(HoldVisibilityTimer:Timer Handle Structure=@n3) -> (Output_Get:Timer Handle Structure)
   n4 = Cast To FortPlayerController(Object:Wildcard)
   > n5 = Bind Event to On View Target Changed(self:Fort Player Controller Object Reference=@n4) {is=title K2Node_AddDelegate}
   n6 = WaitForAttributeChanged!(
    execute:Exec=[@n5.then, @n20.then]
    TargetActor:Actor Object Reference=GetViewTarget!(self:Controller Object Reference=GetOwningPlayer()) -> (ReturnValue:Actor Object Reference)
    Attribute:Gameplay Attribute Structure=Get FuelAttribute() -> (FuelAttribute:Gameplay Attribute Structure)
    OnlyTriggerOnce:bool=false
   ) -> (
    then:Exec
    AsyncTaskProxy:Ability Async Wait Attribute Changed Object Reference
    Changed:Exec
    Attribute:Gameplay Attribute Structure
    NewValue:float=~
    OldValue:float=~
   ) {is=title K2Node_AsyncAction}
   > Set FuelChangeListener(FuelChangeListener:Ability Async Wait Attribute Changed Object Reference=@n6) -> (Output_Get:Ability Async Wait Attribute Changed Object Reference)
   n7 = GetFloatAttribute!(
    Actor:Actor Object Reference=GetOwningPlayer()
    Attribute:Gameplay Attribute Structure=Get MaxFuelAttribute() -> (MaxFuelAttribute:Gameplay Attribute Structure)
   ) -> (bSuccessfullyFoundAttribute:bool, ReturnValue:float)
   > UpdateFuelBar!(
    execute
    self
    Fuel:double=GetFloatAttribute(Actor=GetOwningPlayer(), Attribute=Get FuelAttribute()).ReturnValue
    MaxFuel:double=@n7.ReturnValue
   ) -> (then)
   > SetVisibility(self=[Get HoldPrompt_Text1(), Get HoldPrompt_Text2()], InVisibility=Hidden)
   n8 = GetOwningPlayer()
   > n9 = WaitGameplayTagAddToActor!(
    execute:Exec
    TargetActor:Actor Object Reference=@n8
    Tag:Gameplay Tag Structure=(TagName="Gameplay.Action.Player.RocketDrill.Active")
    OnlyTriggerOnce:bool=false
   ) -> (then:Exec, AsyncTaskProxy:Ability Async Wait Gameplay Tag Added Object Reference, Added:Exec) {is=title K2Node_AsyncAction}
   > n10 = WaitGameplayTagRemoveFromActor!(
    execute:Exec
    TargetActor:Actor Object Reference=@n8
    Tag:Gameplay Tag Structure=(TagName="Gameplay.Action.Player.RocketDrill.Active")
    OnlyTriggerOnce:bool=false
   ) -> (then:Exec, AsyncTaskProxy:Ability Async Wait Gameplay Tag Removed Object Reference, Removed:Exec) {is=title K2Node_AsyncAction}
   > n11 = WaitGameplayEventToActor!(
    execute:Exec
    TargetActor:Actor Object Reference=@n8
    EventTag:Gameplay Tag Structure=(TagName="Event.Athena.Chainsaw.DidntHoldScooterButton")
    OnlyTriggerOnce:bool=false
    OnlyMatchExact:bool=true
   ) -> (then:Exec, AsyncTaskProxy:Ability Async Wait Gameplay Event Object Reference, EventReceived:Exec, Payload:Gameplay Event Data Structure) {is=title K2Node_AsyncAction}
   n12 = Get Anim_OnBoosting() -> (Anim_OnBoosting:Widget Animation Object Reference)
   PlayAnimation(execute=@n10.Removed, InAnimation=@n12, PlayMode=Reverse)
   n13 = Get M_GasMeter() -> (M_GasMeter:Material Instance Dynamic Object Reference)
   > SetScalarParameterValue(self=@n13, ParameterName=isActiveSwitch, Value=0)
   PlayAnimation(execute=@n9.Added, InAnimation=@n12)
   > SetScalarParameterValue(self=@n13, ParameterName=isActiveSwitch, Value=1)
   UpdateFuelBar(execute=@n6.Changed, Fuel=@n6.NewValue, MaxFuel=@n7.ReturnValue)
   Event Destruct()
   > Parent: Destruct()
   > Cancel(self:Cancellable Async Action Object Reference=Get FuelChangeListener() -> (FuelChangeListener:Ability Async Wait Attribute Changed Object Reference))
   > n14 = Get!(execute:Exec) -> (then:Exec, else:Exec, RefuelingAudioComponent:Audio Component Object Reference)
   > Stop(self:Audio Component Object Reference=@n14)
   Event Construct()
   > Parent: Construct()
   > n15 = GetDynamicMaterial(self=Get CenterFrame() -> (CenterFrame:Image Object Reference))
   > Set M_Ret_Frame(M_Ret_Frame:Material Instance Dynamic Object Reference=@n15) -> (Output_Get:Material Instance Dynamic Object Reference)
   > n16 = GetDynamicMaterial(self=Get CenterDot() -> (CenterDot:Image Object Reference))
   > Set M_Center_Dot(M_Center_Dot:Material Instance Dynamic Object Reference=@n16) -> (Output_Get:Material Instance Dynamic Object Reference)
   > n17 = GetDynamicMaterial(self=Get BottomDot() -> (BottomDot:Image Object Reference))
   > n18 = Set M_Bottom_Dot(M_Bottom_Dot:Material Instance Dynamic Object Reference=@n17) -> (Output_Get:Material Instance Dynamic Object Reference)
   > SetScalarParameterValue(self=@n18, ParameterName=MasterOpacity, Value=0)
   > n19 = GetDynamicMaterial(self=Get GasMeter() -> (GasMeter:Image Object Reference))
   > n20 = Set M_GasMeter(M_GasMeter:Material Instance Dynamic Object Reference=@n19) -> (Output_Get:Material Instance Dynamic Object Reference)
   GetOwningPlayer()
  ]}
 OnCooldownStateChanged {nodes [
   n1 = OnCooldownStateChanged!() -> (then, IsCooldownActive:bool=~) {is=title K2Node_FunctionEntry}
   > Set Is Cooldown Active(Is Cooldown Active:bool=@n1) -> (Output_Get:bool)
   > RefreshReticleVisibility()
  ]}
 RefreshReticleVisibility {nodes [
   RefreshReticleVisibility() {is=title K2Node_FunctionEntry}
   > SetVisibility(self=Get CenterDot(), InVisibility=Select(Option 0=Visible, Option 1=Hidden, Index=Get Is Cooldown Active() -> (Is Cooldown Active:bool)))
  ]}
 UpdateFuelBar {nodes [
   n1 = UpdateFuelBar!() -> (then, Fuel:double=~, MaxFuel:double=~) {is=title K2Node_FunctionEntry}
   n2 = Remap Float!(Input=@n1, OldRange_X:double=0, OldRange_Y:double=@n1.MaxFuel, NewRange=(X=0,Y=1), Clamp=true) -> (Output=~)
   > SetScalarParameterValue(self=Get M_GasMeter(), ParameterName=Progress, Value=@n2)
   > UpdateFuelBarAudio!(execute, self, CurrentFuelBar:double=@n2) -> (then)
   > Set PreviousFuelBarValue(PreviousFuelBarValue:double=@n2) -> (Output_Get:double)
  ]}
 UpdateFuelBarAudio {nodes [
   n1 = UpdateFuelBarAudio!() -> (then, CurrentFuelBar:double=~) {is=title K2Node_FunctionEntry}
   > n2 = Branch(Condition="float > float"(A=@n1, B=Get PreviousFuelBarValue() -> (PreviousFuelBarValue:double)))
   > n3 = Branch(Condition="float >= float"(A=Get CurrentFuelBar() -> (CurrentFuelBar:double), B=1))
   > n4 = Get()
   > Stop(self=@n4)
   > PlaySound2D(Sound=Get RefuelingCompletedSound() -> (RefuelingCompletedSound:Sound Base Object Reference))
   n5 = Get(execute=@n3.else)
   Set Float Parameter!(
    execute=[@n8.Is Valid, @n5.then]
    self:Audio Parameter Controller Interface Interface=Get RefuelingAudioComponent() -> (RefuelingAudioComponent:Audio Component Object Reference)
    InName:Name=Charge
    InFloat:float=Get CurrentFuelBar()
   ) -> (then) {is=title K2Node_CallFunction}
   n6 = SpawnSound2D!(
    execute=@n5.else
    Sound:Sound Base Object Reference=Get RefuelingSound() -> (RefuelingSound:Sound Base Object Reference)
    VolumeMultiplier:float=1
    PitchMultiplier:float=1
    StartTime:float=0
    ConcurrencySettings:Sound Concurrency Object Reference
    bPersistAcrossLevelTransition:bool=false
    bAutoDestroy:bool=true
   ) -> (then, ReturnValue:Audio Component Object Reference)
   > n7 = Set RefuelingAudioComponent(RefuelingAudioComponent:Audio Component Object Reference=@n6) -> (Output_Get:Audio Component Object Reference)
   > n8 = Is Valid(InputObject=@n7)
   n9 = Get(execute=@n2.else)
   > Stop(self=@n9)
  ]}
]
variables [
 FuelAttribute {
  type=GameplayAttribute
  default {
   attribute=/Script/RocketDrillGameplayRuntime.RocketDrillAttributeSet:Fuel
   attributeName=Fuel
   attributeOwner=&RocketDrillAttributeSet
  }
 }
 FuelChangeListener {type=AbilityAsync_WaitAttributeChanged, default=null}
 HoldVisibilityTimer {type=TimerHandle}
 Is Cooldown Active {type=boolean}
 M_Bottom_Dot {type=MaterialInstanceDynamic}
 M_Center_Dot {type=MaterialInstanceDynamic}
 M_GasMeter {type=MaterialInstanceDynamic}
 M_Ret_Frame {type=MaterialInstanceDynamic}
 MaxFuelAttribute {
  type=GameplayAttribute
  default {
   attribute=&RocketDrillAttributeSet:MaxFuel
   attributeName=MaxFuel
   attributeOwner=&RocketDrillAttributeSet
  }
 }
 PreviousFuelBarValue {type=number}
 RefuelingAudioComponent {type=AudioComponent, default=null}
 RefuelingCompletedSound {type=SoundBase, default=/RocketDrillGameplay/Assets/Audio/SC_RocketDrill_UI_Refueling_Complete}
 RefuelingSound {type=SoundBase, default=/RocketDrillGameplay/Assets/Audio/SC_RocketDrill_UI_Refueling}
]
widgets {
 info {
  inheritedWidgetCount=1
  parentClass=/Script/FortniteUI.FortWeaponReticleExtensionWidgetBase
  rootWidgetClass=&Overlay
  widgetCount=10
 }
 widgets [
  Reticle_Root {widgetClassPath=&Overlay, widget=Reticle_Root}
  CenterDot {
   widgetClassPath=&Image
   bIsVariable=true
   parent=Reticle_Root
   widget=CenterDot
   properties {
    brush {
     bIsDynamicallyLoaded=false
     drawAs=Image
     imageType=FullColor
     mirroring=NoMirror
     resourceName=None
     resourceObject=&/M_UI_Ret_RocketDrill_Dot
     tiling=NoTile
     imageSize=<50 50>
     margin=<ltrb 0 0 0 0>
     outlineSettings {
      bUseBrushTransparency=false
      roundingType=HalfHeightRadius
      width=0
      color {colorUseRule=UseColor_Specified, specifiedColor=<rgba 0 0 0 0>}
      cornerRadii=<0 0 0 0>
     }
     tintColor {colorUseRule=UseColor_Specified, specifiedColor=<rgba 1 1 1 1>}
     uVRegion {bIsValid=false, max=<0 0>, min=<0 0>}
    }
    colorAndOpacity=<rgba 1 1 1 0.9>
   }
   slot {class=OverlaySlot, properties {horizontalAlignment=HAlign_Center, verticalAlignment=VAlign_Center}}
  }
  BottomDot {
   widgetClassPath=&Image
   bIsVariable=true
   parent=Reticle_Root
   widget=BottomDot
   properties {
    brush {
     bIsDynamicallyLoaded=false
     drawAs=Image
     imageType=FullColor
     mirroring=NoMirror
     resourceName=None
     resourceObject=&M_UI_Ret_RocketDrill_Dot
     tiling=NoTile
     imageSize=<48 48>
     margin=<ltrb 0 0 0 0>
     outlineSettings {
      bUseBrushTransparency=false
      roundingType=HalfHeightRadius
      width=0
      color {colorUseRule=UseColor_Specified, specifiedColor=<rgba 0 0 0 0>}
      cornerRadii=<0 0 0 0>
     }
     tintColor {colorUseRule=UseColor_Specified, specifiedColor=<rgba 1 1 1 1>}
     uVRegion {bIsValid=false, max=<0 0>, min=<0 0>}
    }
    colorAndOpacity=<rgba 1 1 1 0>
    renderTransform {angle=0, scale=<1 1>, shear=<0 0>, translation=<0 25>}
   }
   slot {class=OverlaySlot, properties {horizontalAlignment=HAlign_Center, verticalAlignment=VAlign_Center}}
  }
  CenterFrame {
   widgetClassPath=&Image
   bIsVariable=true
   parent=Reticle_Root
   widget=CenterFrame
   properties {
    brush {
     bIsDynamicallyLoaded=false
     drawAs=Image
     imageType=FullColor
     mirroring=NoMirror
     resourceName=None
     resourceObject=&/M_UI_Ret_RocketDrill_Frame
     tiling=NoTile
     imageSize=<150 150>
     margin=<ltrb 0 0 0 0>
     outlineSettings {
      bUseBrushTransparency=false
      roundingType=HalfHeightRadius
      width=0
      color {colorUseRule=UseColor_Specified, specifiedColor=<rgba 0 0 0 0>}
      cornerRadii=<0 0 0 0>
     }
     tintColor {colorUseRule=UseColor_Specified, specifiedColor=<rgba 1 1 1 1>}
     uVRegion {bIsValid=false, max=<0 0>, min=<0 0>}
    }
    colorAndOpacity=<rgba 1 1 1 0.9>
   }
   slot {class=OverlaySlot, properties {horizontalAlignment=HAlign_Center, verticalAlignment=VAlign_Center}}
  }
  Overlay_TempReorg {
   widgetClassPath=&Overlay
   parent=Reticle_Root
   widget=Overlay_TempReorg
   properties {renderTransform {angle=0, scale=<1 1>, shear=<0 0>, translation=<65 20>}}
   slot {class=OverlaySlot}
  }
  GasMeter {
   widgetClassPath=&Image
   bIsVariable=true
   parent=Overlay_TempReorg
   widget=GasMeter
   properties {
    brush {
     bIsDynamicallyLoaded=false
     drawAs=Image
     imageType=FullColor
     mirroring=NoMirror
     resourceName=None
     resourceObject=&/M_UI_Ret_RocketDrill_Fuel
     tiling=NoTile
     imageSize=<200 200>
     margin=<ltrb 0 0 0 0>
     outlineSettings {
      bUseBrushTransparency=false
      roundingType=HalfHeightRadius
      width=0
      color {colorUseRule=UseColor_Specified, specifiedColor=<rgba 0 0 0 0>}
      cornerRadii=<0 0 0 0>
     }
     tintColor {colorUseRule=UseColor_Specified, specifiedColor=<rgba 1 1 1 1>}
     uVRegion {bIsValid=false, max=<0 0>, min=<0 0>}
    }
    colorAndOpacity=<rgba 1 1 1 0.9>
   }
   slot {class=OverlaySlot}
  }
  HoldPrompt {
   widgetClassPath=&HorizontalBox
   parent=Overlay_TempReorg
   widget=HoldPrompt
   properties {renderTransform {angle=0, scale=<1 1>, shear=<0 0>, translation=<200 59>}}
   slot {class=OverlaySlot, properties {horizontalAlignment=HAlign_Right, verticalAlignment=VAlign_Center}}
  }
  HoldPrompt_Text1 {
   widgetClassPath=&CommonTextBlock
   bIsVariable=true
   parent=HoldPrompt
   widget=HoldPrompt_Text1
   properties {
    style=/FortUILibrary/Styles/Typography/_TextStyle_TagNormal._TextStyle_TagNormal_C
    text="Hold "
    visibility=Hidden
    font {
     bForceMonospaced=false
     bMaterialIsStencil=false
     fontMaterial=null
     fontObject=&BurbankSmall
     letterSpacing=0
     monospacedWidth=1
     size=15
     skewAmount=0
     typefaceFontName=Black
     outlineSettings {
      bApplyOutlineToDropShadows=false
      bMiteredCorners=false
      bSeparateFillAlpha=false
      outlineMaterial=null
      outlineSize=0
      outlineColor=<rgba 0 0 0 1>
     }
    }
    shadowOffset=<0 0>
    strikeBrush {
     bIsDynamicallyLoaded=false
     drawAs=NoDrawType
     imageType=NoImage
     mirroring=NoMirror
     resourceName=None
     resourceObject=null
     tiling=NoTile
     imageSize=<32 32>
     margin=<ltrb 0 0 0 0>
     outlineSettings {
      bUseBrushTransparency=false
      roundingType=HalfHeightRadius
      width=0
      color {colorUseRule=UseColor_Specified, specifiedColor=<rgba 0 0 0 0>}
      cornerRadii=<0 0 0 0>
     }
     tintColor {colorUseRule=UseColor_Specified, specifiedColor=<rgba 1 1 1 1>}
     uVRegion {bIsValid=false, max=<0 0>, min=<0 0>}
    }
   }
   slot {class=HorizontalBoxSlot, properties {verticalAlignment=VAlign_Center, padding=<ltrb 4 0 0 0>}}
  }
  HoldPrompt_Text2 {
   widgetClassPath=&CommonTextBlock
   bIsVariable=true
   parent=HoldPrompt
   widget=HoldPrompt_Text2
   properties {
    style=&_TextStyle_TagNormal_C
    text=to Travel Further
    visibility=Hidden
    font {
     bForceMonospaced=false
     bMaterialIsStencil=false
     fontMaterial=null
     fontObject=&BurbankSmall
     letterSpacing=0
     monospacedWidth=1
     size=15
     skewAmount=0
     typefaceFontName=Black
     outlineSettings {
      bApplyOutlineToDropShadows=false
      bMiteredCorners=false
      bSeparateFillAlpha=false
      outlineMaterial=null
      outlineSize=0
      outlineColor=<rgba 0 0 0 1>
     }
    }
    shadowOffset=<0 0>
    strikeBrush {
     bIsDynamicallyLoaded=false
     drawAs=NoDrawType
     imageType=NoImage
     mirroring=NoMirror
     resourceName=None
     resourceObject=null
     tiling=NoTile
     imageSize=<32 32>
     margin=<ltrb 0 0 0 0>
     outlineSettings {
      bUseBrushTransparency=false
      roundingType=HalfHeightRadius
      width=0
      color {colorUseRule=UseColor_Specified, specifiedColor=<rgba 0 0 0 0>}
      cornerRadii=<0 0 0 0>
     }
     tintColor {colorUseRule=UseColor_Specified, specifiedColor=<rgba 1 1 1 1>}
     uVRegion {bIsValid=false, max=<0 0>, min=<0 0>}
    }
   }
   slot {class=HorizontalBoxSlot, properties {verticalAlignment=VAlign_Center, padding=<ltrb 4 0 0 0>}}
  }
  TutorialHighlight {widgetClassPath=&TutorialHighlight}
 ]
}
